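# Physics asset for the dynamic back skirt of the female medium Butterfly character body.
# Three short back chains of box bodies hang from the pelvis with twist locked and limited swing.
# Two five-bone side chains, left and right, also root at the pelvis, with twist limits mostly tight.
# Pelvis and thigh capsules plus calf spheres provide body shapes for skirt collision, with leg constraints left fully free.
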PhysicsAsset /BRCosmetics/Characters/Player/Female/Medium/Bodies/F_MED_Butterfly/Meshes/F_MED_Butterfly_BackSkirt_Dyn_Skirt
bodies {
 calf_l [{shapeType=Sphere, radius=5.5445}]
 calf_r [{shapeType=Sphere, radius=5.5445}]
 dyn_skirtback_00_l [{shapeType=Box, extentX=5.5465, extentY=3.1323, extentZ=15.542, center=<3.1021 0.0002 0.0002>, rotation=<rot -0.0006 0.0001 110>}]
 dyn_skirtback_00_r [{shapeType=Box, extentX=5.5465, extentY=3.1323, extentZ=15.542, center=<3.1021 -0.0002 -0.0002>, rotation=<rot -0.0007 -0.0002 70>}]
 dyn_skirtback_01 [{shapeType=Box, extentX=6.0465, extentY=4.6323, extentZ=14.542, center=<3.8203 0.0002 0.0002>, rotation=<rot 0 0 90>}]
 dyn_skirtback_01_l [{shapeType=Box, extentX=6.5465, extentY=3.6323, extentZ=17.042, center=<3.1021 0.0002 0.0002>, rotation=<rot -0.0006 0.0001 110>}]
 dyn_skirtback_01_r [{shapeType=Box, extentX=5.5465, extentY=3.6323, extentZ=17.042, center=<3.1021 0.0002 0.0002>, rotation=<rot -0.0007 -0.0002 70>}]
 dyn_skirtback_02 [{shapeType=Box, extentX=6.0465, extentY=5.1323, extentZ=14.542, center=<3.1021 0.0002 0.0002>, rotation=<rot 0 0 90>}]
 dyn_skirtback_base [{shapeType=Box, extentX=6.0465, extentY=3.6323, extentZ=15.542, center=<3.1021 0.0002 0.0002>, rotation=<rot 0 0 90>}]
 dyn_skirtback_base_l [{shapeType=Box, extentX=6.5465, extentY=2.3823, extentZ=16.542, center=<3.1021 0.0002 0.0002>, rotation=<rot -0.0006 0.0001 110>}]
 dyn_skirtback_base_r [{shapeType=Box, extentX=6.5465, extentY=2.3823, extentZ=16.542, center=<3.1021 -0.0002 -0.0002>, rotation=<rot -0.0007 -0.0002 70>}]
 dyn_skirtsideB_00_l [{shapeType=Box, extentX=8.4215, extentY=16.2573, extentZ=1.3545, center=<3.9692 -1.3464 0>, rotation=<rot 0 0 -5>}]
 dyn_skirtsideB_00_r [{shapeType=Box, extentX=8.4215, extentY=16.2573, extentZ=1.3545, center=<3.9692 -1.3464 0>, rotation=<rot 0 0 5>}]
 dyn_skirtsideB_01_l [{shapeType=Box, extentX=8.1715, extentY=16.2573, extentZ=1.3545, center=<3.2463 -1.3428 0.0986>, rotation=<rot 0 0 -0.8007>}]
 dyn_skirtsideB_01_r [{shapeType=Box, extentX=6.7965, extentY=16.2573, extentZ=1.3545, center=<3.2463 -1.3464 0>, rotation=<rot 0 0 5>}]
 dyn_skirtsideB_02_l [{shapeType=Box, extentX=7.1715, extentY=16.2573, extentZ=1.3545, center=<3.1828 -1.3442 0.0757>, rotation=<rot 0 0 -1.7731>}]
 dyn_skirtsideB_02_r [{shapeType=Box, extentX=7.1715, extentY=16.2573, extentZ=1.3545, center=<3.1828 -1.3464 -0.0001>, rotation=<rot 0 0 5>}]
 dyn_skirtsideB_03_l [
  {shapeType=Box, extentX=7.109, extentY=23.8198, extentZ=2.917, center=<3.0214 -0.1386 -0.2475>, rotation=<rot -2.1966 -3.1905 10.3998>}
  {shapeType=Box, extentX=3.625, extentY=9.9375, extentZ=3, center=<4.8047 6.7002 -3.2196>, rotation=<rot -3.8078 -10.6939 36.2466>}
 ]
 dyn_skirtsideB_03_r [
  {shapeType=Box, extentX=7.109, extentY=23.8198, extentZ=2.917, center=<3.0467 -0.0443 0.2738>, rotation=<rot 7.007 -0.7616 -10.4701>}
  {shapeType=Box, extentX=3.625, extentY=9.9375, extentZ=3, center=<4.8047 6.7002 3.2196>, rotation=<rot 10.661 0 -36.8931>}
 ]
 dyn_skirtsideB_04_l [{shapeType=Box, extentX=2.4215, extentY=8.2573, extentZ=1.917, center=<0.9086 2.4997 -0.0002>}]
 dyn_skirtsideB_04_r [{shapeType=Box, extentX=2.4215, extentY=8.2573, extentZ=1.917, center=<0.9086 2.4997 -0.0002>}]
 pelvis [
  {
   shapeType=Capsule
   length=11.4785
   radius=10.8878
   shapeName=backSkirtCOL
   center=<-4.0438 -2.691 -0.0209>
   rotation=<rot -0.1695 -90 -0.4083>
  }
  {shapeType=Capsule, length=11.1347, radius=9.0753, shapeName=frontSkirtCOL, center=<-4.732 -1.4329 0>}
 ]
 thigh_l [{shapeType=Capsule, length=23.291, radius=10.8991, center=<-15.1582 -0.582 1.5021>, rotation=<rot 83.3096 174.9975 0>}]
 thigh_r [{shapeType=Capsule, length=27.041, radius=10.8991, center=<15.1582 0.582 -1.5021>, rotation=<rot 83.3096 174.9975 180>}]
 thigh_twist_01_l [{shapeType=Capsule, length=6.791, radius=10.6491, center=<10.283 0.0045 2.32>, rotation=<rot 84.9756 -179.9999 -174.9999>}]
 thigh_twist_01_r [{shapeType=Capsule, length=6.791, radius=10.6491, center=<-10.283 -0.0045 -2.32>, rotation=<rot 84.9757 -179.9999 5.0001>}]
}
constraints [
 {
  bone1Name=dyn_skirtback_02
  bone2Name=dyn_skirtback_01
  swing1LimitDegrees=20
  swing2LimitDegrees=25
  twistLimitDegrees=20
  twistMotion=Locked
 }
 {
  bone1Name=dyn_skirtback_01_l
  bone2Name=dyn_skirtback_00_l
  swing1LimitDegrees=20
  swing2LimitDegrees=20
  twistLimitDegrees=20
  twistMotion=Locked
 }
 {
  bone1Name=dyn_skirtback_00_l
  bone2Name=dyn_skirtback_base_l
  swing1LimitDegrees=20
  swing2LimitDegrees=20
  twistLimitDegrees=20
  twistMotion=Locked
 }
 {
  bone1Name=dyn_skirtback_01
  bone2Name=dyn_skirtback_base
  swing1LimitDegrees=20
  swing2LimitDegrees=25
  twistLimitDegrees=20
  twistMotion=Locked
 }
 {
  bone1Name=dyn_skirtback_00_r
  bone2Name=dyn_skirtback_base_r
  swing1LimitDegrees=20
  swing2LimitDegrees=20
  twistLimitDegrees=20
  twistMotion=Locked
 }
 {
  bone1Name=dyn_skirtback_01_r
  bone2Name=dyn_skirtback_00_r
  swing1LimitDegrees=20
  swing2LimitDegrees=20
  twistLimitDegrees=20
  twistMotion=Locked
 }
 {bone1Name=dyn_skirtback_base, bone2Name=pelvis, swing1LimitDegrees=20, swing2LimitDegrees=25, twistLimitDegrees=20, twistMotion=Locked}
 {bone1Name=dyn_skirtback_base_r, bone2Name=pelvis, swing1LimitDegrees=20, swing2LimitDegrees=20, twistLimitDegrees=20, twistMotion=Locked}
 {bone1Name=dyn_skirtback_base_l, bone2Name=pelvis, swing1LimitDegrees=20, swing2LimitDegrees=20, twistLimitDegrees=20, twistMotion=Locked}
 {
  bone1Name=thigh_l
  bone2Name=pelvis
  swing1LimitDegrees=45
  swing1Motion=Free
  swing2LimitDegrees=45
  swing2Motion=Free
  twistLimitDegrees=45
  twistMotion=Free
 }
 {
  bone1Name=calf_l
  bone2Name=thigh_l
  swing1LimitDegrees=45
  swing1Motion=Free
  swing2LimitDegrees=45
  swing2Motion=Free
  twistLimitDegrees=45
  twistMotion=Free
 }
 {
  bone1Name=calf_r
  bone2Name=thigh_r
  swing1LimitDegrees=45
  swing1Motion=Free
  swing2LimitDegrees=45
  swing2Motion=Free
  twistLimitDegrees=45
  twistMotion=Free
 }
 {bone1Name=dyn_skirtsideB_00_r, bone2Name=pelvis, swing1LimitDegrees=45, swing2LimitDegrees=45, twistLimitDegrees=5}
 {bone1Name=dyn_skirtsideB_01_r, bone2Name=dyn_skirtsideB_00_r, swing1LimitDegrees=45, swing2LimitDegrees=45, twistLimitDegrees=5}
 {bone1Name=dyn_skirtsideB_02_r, bone2Name=dyn_skirtsideB_01_r, swing1LimitDegrees=45, swing2LimitDegrees=45, twistLimitDegrees=5}
 {bone1Name=dyn_skirtsideB_03_r, bone2Name=dyn_skirtsideB_02_r, swing1LimitDegrees=45, swing2LimitDegrees=45, twistLimitDegrees=5}
 {bone1Name=dyn_skirtsideB_04_r, bone2Name=dyn_skirtsideB_03_r, swing1LimitDegrees=30, swing2LimitDegrees=45, twistLimitDegrees=45}
 {bone1Name=dyn_skirtsideB_00_l, bone2Name=pelvis, swing1LimitDegrees=45, swing2LimitDegrees=45, twistLimitDegrees=5}
 {bone1Name=dyn_skirtsideB_01_l, bone2Name=dyn_skirtsideB_00_l, swing1LimitDegrees=45, swing2LimitDegrees=45, twistLimitDegrees=5}
 {bone1Name=dyn_skirtsideB_02_l, bone2Name=dyn_skirtsideB_01_l, swing1LimitDegrees=45, swing2LimitDegrees=45, twistLimitDegrees=5}
 {bone1Name=dyn_skirtsideB_03_l, bone2Name=dyn_skirtsideB_02_l, swing1LimitDegrees=45, swing2LimitDegrees=45, twistLimitDegrees=5}
 {bone1Name=dyn_skirtsideB_04_l, bone2Name=dyn_skirtsideB_03_l, swing1LimitDegrees=30, swing2LimitDegrees=45, twistLimitDegrees=45}
 {
  bone1Name=thigh_twist_01_l
  bone2Name=thigh_l
  swing1LimitDegrees=45
  swing1Motion=Free
  swing2LimitDegrees=45
  swing2Motion=Free
  twistLimitDegrees=45
  twistMotion=Free
 }
 {
  bone1Name=thigh_r
  bone2Name=pelvis
  swing1LimitDegrees=45
  swing1Motion=Free
  swing2LimitDegrees=45
  swing2Motion=Free
  twistLimitDegrees=45
  twistMotion=Free
 }
 {
  bone1Name=thigh_twist_01_r
  bone2Name=thigh_r
  swing1LimitDegrees=45
  swing1Motion=Free
  swing2LimitDegrees=45
  swing2Motion=Free
  twistLimitDegrees=45
  twistMotion=Free
 }
 {
  bone1Name=thigh_twist_01_l
  bone2Name=thigh_twist_01_r
  swing1LimitDegrees=45
  swing1Motion=Free
  swing2LimitDegrees=45
  swing2Motion=Free
  twistLimitDegrees=45
  twistMotion=Free
 }
]
