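# Physics asset for the front loincloth of the F_MED_Butterfly female player body, setting up its dynamic cloth-like swing against the lower body.
# Three box bodies on the dyn_skirt_front bones form a limited-angle chain from the pelvis, with tight twist and one tight swing axis at the root.
# Named pelvis capsules and a box, thigh capsules and calf spheres supply collision shapes, with leg joints left free in swing and twist.
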PhysicsAsset /BRCosmetics/Characters/Player/Female/Medium/Bodies/F_MED_Butterfly/Meshes/F_MED_Butterfly_FrontLoincloth_Dyn_Loincloth_Front
bodies {
 calf_l [{shapeType=Sphere, radius=5.5445}]
 calf_r [{shapeType=Sphere, radius=5.5445}]
 dyn_skirt_front_00 [{shapeType=Box, extentX=9.5465, extentY=4, extentZ=15.042, center=<5.3147 -0.0004 0.0003>}]
 dyn_skirt_front_01 [{shapeType=Box, extentX=9.5465, extentY=4, extentZ=15.042, center=<5.3147 -0.0004 0.0003>}]
 dyn_skirt_front_02 [{shapeType=Box, extentX=9.5465, extentY=4, extentZ=15.042, center=<5.3147 -0.0004 0.0003>}]
 pelvis [
  {shapeType=Capsule, length=6.7285, radius=8.6378, shapeName=backSkirtCOL, center=<2.5729 -2.2848 0.0202>, rotation=<rot -0.1695 -90 -0.4083>}
  {shapeType=Capsule, length=7.4785, radius=8.1378, shapeName=frontSkirtCOL, center=<-0.8307 -1.5792 0>}
  {shapeType=Box, extentX=62, extentY=14, extentZ=23.5, shapeName=lowerFrontSkirtBoxCOL, center=<-2.2042 -1.6958 -0.0001>}
 ]
 thigh_l [{shapeType=Capsule, length=32.541, radius=7.7741, center=<-18.6303 -0.2781 0.0002>, rotation=<rot 90 3.5763 8.5763>}]
 thigh_r [{shapeType=Capsule, length=32.541, radius=7.7741, center=<18.6303 0.2781 -0.0002>, rotation=<rot 90 -90 95>}]
}
constraints [
 {bone1Name=dyn_skirt_front_01, bone2Name=dyn_skirt_front_00, swing1LimitDegrees=30, swing2LimitDegrees=30, twistLimitDegrees=20}
 {bone1Name=dyn_skirt_front_02, bone2Name=dyn_skirt_front_01, swing1LimitDegrees=30, swing2LimitDegrees=30, twistLimitDegrees=20}
 {bone1Name=dyn_skirt_front_00, bone2Name=pelvis, swing1LimitDegrees=45, swing2LimitDegrees=5, twistLimitDegrees=5}
 {
  bone1Name=thigh_l
  bone2Name=pelvis
  swing1LimitDegrees=45
  swing1Motion=Free
  swing2LimitDegrees=45
  swing2Motion=Free
  twistLimitDegrees=45
  twistMotion=Free
 }
 {
  bone1Name=calf_l
  bone2Name=thigh_l
  swing1LimitDegrees=45
  swing1Motion=Free
  swing2LimitDegrees=45
  swing2Motion=Free
  twistLimitDegrees=45
  twistMotion=Free
 }
 {
  bone1Name=calf_r
  bone2Name=thigh_r
  swing1LimitDegrees=45
  swing1Motion=Free
  swing2LimitDegrees=45
  swing2Motion=Free
  twistLimitDegrees=45
  twistMotion=Free
 }
 {
  bone1Name=thigh_r
  bone2Name=pelvis
  swing1LimitDegrees=45
  swing1Motion=Free
  swing2LimitDegrees=45
  swing2Motion=Free
  twistLimitDegrees=45
  twistMotion=Free
 }
]
