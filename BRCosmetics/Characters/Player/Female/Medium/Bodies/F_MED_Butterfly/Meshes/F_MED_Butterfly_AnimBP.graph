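# Post-process-style Anim Blueprint for the female Butterfly cosmetic body, copying the parent mesh pose and adding secondary motion to arms, collar and skirt.
# The cached collar pose feeds two branches blended by rigidsOnOffBool: RigidBody loincloth and back-skirt simulation, or a thigh-driven constraint fallback.
# Anim Dynamics and additive bone offsets, weighted by crouch, sprint and neck-targeting alphas, refine the result.
# Several side and back skirt chains are left unconnected.
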AnimBlueprint /BRCosmetics/Characters/Player/Female/Medium/Bodies/F_MED_Butterfly/Meshes/F_MED_Butterfly_AnimBP
target_skeleton=&/F_MED_Butterfly_Skeleton
graphs [
 AnimGraph {nodes [
   n1 = Anim Dynamics - Bone: dyn_skirtsideB_00_r(Alpha=1, ExternalForce="0,0,0") {
    angularDampingOverride=1
    angularSpringConstant=50
    bAngularSpring=true
    bGravityOverrideInSimSpace=true
    bOverrideAngularDamping=true
    gravityScale=0.1
    simulationSpace=RootRelative
    alphaBoolBlend {blendOption=Linear}
    boundBone {boneName=dyn_skirtsideB_00_r}
    componentAppliedLinearAccClamp=<500 500 500>
    componentLinearAccScale=<0.6 0.6 0.6>
    componentLinearVelScale=<1.7 1.7 1.7>
    physicsBodyDefinitions [{
      sphereCollisionRadius=0
      boundBone {boneName=dyn_skirtsideB_00_r}
      boxExtents=<6 12 2>
      constraintSetup {angularLimitsMax=<10 35 5>, angularLimitsMin=<-10 0 -15>, angularTarget=<1 0 0>}
      localJointOffset=<1 0 0>
     }]
    relativeSpaceBone {boneName=pelvis}
   }
   n2 = Anim Dynamics - Bone: dyn_skirtsideB_01_r(ComponentPose=@n1, Alpha=1, ExternalForce="0,0,0") {
    angularDampingOverride=1
    angularSpringConstant=50
    bAngularSpring=true
    bGravityOverrideInSimSpace=true
    bOverrideAngularDamping=true
    gravityScale=0.35
    simulationSpace=RootRelative
    alphaBoolBlend {blendOption=Linear}
    boundBone {boneName=dyn_skirtsideB_01_r}
    componentAppliedLinearAccClamp=<500 500 150>
    componentLinearAccScale=<0.6 0.6 0.6>
    componentLinearVelScale=<1.7 1.7 1.7>
    physicsBodyDefinitions [{
      sphereCollisionRadius=0
      boundBone {boneName=dyn_skirtsideB_01_r}
      boxExtents=<6 12 2>
      constraintSetup {angularLimitsMax=<10 50 5>, angularLimitsMin=<-10 0 -5>, angularTarget=<1 0 0>}
      localJointOffset=<1 0 0>
     }]
    relativeSpaceBone {boneName=pelvis}
   }
   n3 = Anim Dynamics - Bone: dyn_skirtsideB_02_r(ComponentPose=@n2, Alpha=0, ExternalForce="0,0,0") {
    angularDampingOverride=1
    angularSpringConstant=50
    bAngularSpring=true
    bGravityOverrideInSimSpace=true
    bOverrideAngularDamping=true
    gravityScale=0.35
    simulationSpace=RootRelative
    alphaBoolBlend {blendOption=Linear}
    boundBone {boneName=dyn_skirtsideB_02_r}
    chainEnd {boneName=dyn_skirtsideB_03_l}
    componentAppliedLinearAccClamp=<500 500 150>
    componentLinearAccScale=<0.6 0.6 0.6>
    componentLinearVelScale=<1.7 1.7 1.7>
    physicsBodyDefinitions [{
      sphereCollisionRadius=0
      boundBone {boneName=dyn_skirtsideB_02_r}
      boxExtents=<6 12 2>
      constraintSetup {angularLimitsMax=<0 25 15>, angularLimitsMin=<0 0 0>, angularTarget=<1 0 0>}
      localJointOffset=<1 0 0>
     }]
    relativeSpaceBone {boneName=pelvis}
   }
   n4 = Anim Dynamics - Bone: dyn_skirtsideB_03_r(ComponentPose=@n3, Alpha=1, ExternalForce="0,0,0") {
    angularDampingOverride=1
    angularSpringConstant=50
    bAngularSpring=true
    bGravityOverrideInSimSpace=true
    bOverrideAngularDamping=true
    gravityScale=0.35
    simulationSpace=RootRelative
    alphaBoolBlend {blendOption=Linear}
    boundBone {boneName=dyn_skirtsideB_03_r}
    chainEnd {boneName=dyn_skirtsideB_04_r}
    componentAppliedLinearAccClamp=<500 500 150>
    componentLinearAccScale=<0.6 0.6 0.6>
    componentLinearVelScale=<1.7 1.7 1.7>
    physicsBodyDefinitions [{
      sphereCollisionRadius=0
      boundBone {boneName=dyn_skirtsideB_03_r}
      boxExtents=<6 12 12>
      constraintSetup {angularLimitsMax=<0 25 35>, angularLimitsMin=<0 0 0>, angularTarget=<1 0 0>}
      localJointOffset=<1 0 0>
     }]
    relativeSpaceBone {boneName=pelvis}
   }
   Anim Dynamics - Bone: dyn_skirtsideB_04_r(ComponentPose=@n4, Alpha=1, ExternalForce="0,0,0") {
    angularDampingOverride=1
    angularSpringConstant=50
    bAngularSpring=true
    bGravityOverrideInSimSpace=true
    bOverrideAngularDamping=true
    gravityScale=0.4
    simulationSpace=RootRelative
    alphaBoolBlend {blendOption=Linear}
    boundBone {boneName=dyn_skirtsideB_04_r}
    componentAppliedLinearAccClamp=<500 500 150>
    componentLinearAccScale=<0.6 0.6 0.6>
    componentLinearVelScale=<1.7 1.7 1.7>
    physicsBodyDefinitions [{
      sphereCollisionRadius=0
      boundBone {boneName=dyn_skirtsideB_04_r}
      boxExtents=<6 12 2>
      constraintSetup {angularLimitsMax=<25 25 35>, angularLimitsMin=<0 0 0>, angularTarget=<1 0 0>}
      localJointOffset=<0 2 0>
     }]
    relativeSpaceBone {boneName=pelvis}
   }
   n5 = Anim Dynamics - Bone: dyn_skirtsideB_00_r(Alpha=1, ExternalForce="0,0,0") {
    angularDampingOverride=1
    angularSpringConstant=50
    bAngularSpring=true
    bGravityOverrideInSimSpace=true
    bOverrideAngularDamping=true
    gravityScale=0.1
    simulationSpace=RootRelative
    alphaBoolBlend {blendOption=Linear}
    boundBone {boneName=dyn_skirtsideB_00_r}
    componentAppliedLinearAccClamp=<500 500 500>
    componentLinearAccScale=<0.7 0.7 0.7>
    componentLinearVelScale=<1.6 1.6 1.6>
    physicsBodyDefinitions [{
      sphereCollisionRadius=0
      boundBone {boneName=dyn_skirtsideB_00_r}
      boxExtents=<6 12 2>
      constraintSetup {angularLimitsMax=<10 15 5>, angularLimitsMin=<-10 0 -15>, angularTarget=<1 0 0>}
      localJointOffset=<1 0 0>
     }]
    relativeSpaceBone {boneName=pelvis}
   }
   n6 = Anim Dynamics - Bone: dyn_skirtsideB_01_r(ComponentPose=@n5, Alpha=1, ExternalForce="0,0,0") {
    angularDampingOverride=1
    angularSpringConstant=50
    bAngularSpring=true
    bGravityOverrideInSimSpace=true
    bOverrideAngularDamping=true
    gravityScale=0.35
    simulationSpace=RootRelative
    alphaBoolBlend {blendOption=Linear}
    boundBone {boneName=dyn_skirtsideB_01_r}
    componentAppliedLinearAccClamp=<500 500 150>
    componentLinearAccScale=<0.7 0.7 0.7>
    componentLinearVelScale=<1.6 1.6 1.6>
    physicsBodyDefinitions [{
      sphereCollisionRadius=0
      boundBone {boneName=dyn_skirtsideB_01_r}
      boxExtents=<6 12 2>
      constraintSetup {angularLimitsMax=<10 35 5>, angularLimitsMin=<-10 0 -5>, angularTarget=<1 0 0>}
      localJointOffset=<1 0 0>
     }]
    relativeSpaceBone {boneName=pelvis}
   }
   n7 = Anim Dynamics - Bone: dyn_skirtsideB_02_r(ComponentPose=@n6, Alpha=1, ExternalForce="0,0,0") {
    angularDampingOverride=1
    angularSpringConstant=50
    bAngularSpring=true
    bGravityOverrideInSimSpace=true
    bOverrideAngularDamping=true
    gravityScale=0.35
    simulationSpace=RootRelative
    alphaBoolBlend {blendOption=Linear}
    boundBone {boneName=dyn_skirtsideB_02_r}
    chainEnd {boneName=dyn_skirtsideB_03_l}
    componentAppliedLinearAccClamp=<500 500 150>
    componentLinearAccScale=<0.7 0.7 0.7>
    componentLinearVelScale=<1.6 1.6 1.6>
    physicsBodyDefinitions [{
      sphereCollisionRadius=0
      boundBone {boneName=dyn_skirtsideB_02_r}
      boxExtents=<6 12 2>
      constraintSetup {angularLimitsMax=<0 35 15>, angularLimitsMin=<0 0 0>, angularTarget=<1 0 0>}
      localJointOffset=<1 0 0>
     }]
    relativeSpaceBone {boneName=pelvis}
   }
   n8 = Anim Dynamics - Bone: dyn_skirtsideB_03_r(ComponentPose=@n7, Alpha=1, ExternalForce="0,0,0") {
    angularDampingOverride=1
    angularSpringConstant=50
    bAngularSpring=true
    bGravityOverrideInSimSpace=true
    bOverrideAngularDamping=true
    gravityScale=0.35
    simulationSpace=RootRelative
    alphaBoolBlend {blendOption=Linear}
    boundBone {boneName=dyn_skirtsideB_03_r}
    componentAppliedLinearAccClamp=<500 500 150>
    componentLinearAccScale=<0.7 0.7 0.7>
    componentLinearVelScale=<1.6 1.6 1.6>
    physicsBodyDefinitions [{
      sphereCollisionRadius=0
      boundBone {boneName=dyn_skirtsideB_03_r}
      boxExtents=<6 12 2>
      constraintSetup {angularLimitsMax=<0 15 35>, angularLimitsMin=<0 0 0>, angularTarget=<1 0 0>}
      localJointOffset=<1 0 0>
     }]
    relativeSpaceBone {boneName=pelvis}
   }
   Anim Dynamics - Bone: dyn_skirtsideB_04_r(ComponentPose=@n8, Alpha=1, ExternalForce="0,0,0") {
    angularDampingOverride=1
    angularSpringConstant=50
    bAngularSpring=true
    bGravityOverrideInSimSpace=true
    bOverrideAngularDamping=true
    gravityScale=0.4
    simulationSpace=RootRelative
    alphaBoolBlend {blendOption=Linear}
    boundBone {boneName=dyn_skirtsideB_04_r}
    componentAppliedLinearAccClamp=<500 500 150>
    componentLinearAccScale=<0.7 0.7 0.7>
    componentLinearVelScale=<1.6 1.6 1.6>
    physicsBodyDefinitions [{
      sphereCollisionRadius=0
      boundBone {boneName=dyn_skirtsideB_04_r}
      boxExtents=<6 12 2>
      constraintSetup {angularLimitsMax=<25 15 35>, angularLimitsMin=<0 0 0>, angularTarget=<1 0 0>}
      localJointOffset=<0 1 0>
     }]
    relativeSpaceBone {boneName=pelvis}
   }
   n9 = Anim Dynamics - Bone: dyn_skirtback_02_l(Alpha=1, ExternalForce="0,0,0") {
    angularDampingOverride=0.9999
    angularSpringConstant=50
    bAngularSpring=true
    bGravityOverrideInSimSpace=true
    bOverrideAngularDamping=true
    gravityScale=0.35
    simulationSpace=RootRelative
    alphaBoolBlend {blendOption=Linear}
    boundBone {boneName=dyn_skirtback_02_l}
    componentAppliedLinearAccClamp=<1000 1000 1000>
    componentLinearAccScale=<55 55 55>
    componentLinearVelScale=<55 55 55>
    physicsBodyDefinitions [{
      sphereCollisionRadius=0
      boundBone {boneName=dyn_skirtback_02_l}
      boxExtents=<6 12 2>
      constraintSetup {angularLimitsMax=<3 75 0>, angularLimitsMin=<-3 -75 0>, angularTarget=<1 0 0>}
      localJointOffset=<1 0 0>
     }]
    relativeSpaceBone {boneName=dyn_skirtback_03}
   }
   n10 = Anim Dynamics - Bone: dyn_skirtback_02_r(ComponentPose=@n9, Alpha=1, ExternalForce="0,0,0") {
    angularDampingOverride=0.9999
    angularSpringConstant=50
    bAngularSpring=true
    bGravityOverrideInSimSpace=true
    bOverrideAngularDamping=true
    gravityScale=0.35
    simulationSpace=RootRelative
    alphaBoolBlend {blendOption=Linear}
    boundBone {boneName=dyn_skirtback_02_r}
    componentAppliedLinearAccClamp=<1000 1000 1000>
    componentLinearAccScale=<55 55 55>
    componentLinearVelScale=<55 55 55>
    physicsBodyDefinitions [{
      sphereCollisionRadius=0
      boundBone {boneName=dyn_skirtback_02_r}
      boxExtents=<6 12 2>
      constraintSetup {angularLimitsMax=<3 75 0>, angularLimitsMin=<-3 -75 0>, angularTarget=<1 0 0>}
      localJointOffset=<1 0 0>
     }]
    relativeSpaceBone {boneName=dyn_skirtback_03}
   }
   Anim Dynamics - Bone: dyn_skirtback_03(ComponentPose=@n10, Alpha=1, ExternalForce="0,0,0") {
    angularDampingOverride=0.9999
    angularSpringConstant=50
    bAngularSpring=true
    bGravityOverrideInSimSpace=true
    bOverrideAngularDamping=true
    gravityScale=0.35
    simulationSpace=RootRelative
    alphaBoolBlend {blendOption=Linear}
    boundBone {boneName=dyn_skirtback_03}
    componentAppliedLinearAccClamp=<1000 1000 1000>
    componentLinearAccScale=<55 55 55>
    componentLinearVelScale=<55 55 55>
    physicsBodyDefinitions [{
      sphereCollisionRadius=0
      boundBone {boneName=dyn_skirtback_03}
      boxExtents=<6 12 2>
      constraintSetup {angularLimitsMax=<3 75 5>, angularLimitsMin=<-3 -75 -5>, angularTarget=<1 0 0>}
      localJointOffset=<1 0 0>
     }]
    relativeSpaceBone {boneName=dyn_skirtback_03}
   }
   n11 = Anim Dynamics - Bone: dyn_skirtsideB_00_r(Alpha=1, ExternalForce="0,0,0") {
    angularDampingOverride=0.9999
    angularSpringConstant=50
    bAngularSpring=true
    bGravityOverrideInSimSpace=true
    bOverrideAngularDamping=true
    gravityScale=0.1
    simulationSpace=RootRelative
    alphaBoolBlend {blendOption=Linear}
    boundBone {boneName=dyn_skirtsideB_00_r}
    chainEnd {boneName=dyn_skirtsideB_01_r}
    componentAppliedLinearAccClamp=<500 500 500>
    componentLinearAccScale=<25 25 25>
    componentLinearVelScale=<25 25 25>
    physicsBodyDefinitions [{
      sphereCollisionRadius=0
      boundBone {boneName=dyn_skirtsideB_00_r}
      boxExtents=<6 12 2>
      constraintSetup {angularLimitsMax=<10 35 5>, angularLimitsMin=<-10 0 -15>, angularTarget=<1 0 0>}
      localJointOffset=<1 0 0>
     }]
    relativeSpaceBone {boneName=pelvis}
   }
   n12 = Anim Dynamics - Bone: dyn_skirtsideB_01_r(ComponentPose=@n11, Alpha=1, ExternalForce="0,0,0") {
    angularDampingOverride=0.9999
    angularSpringConstant=50
    bAngularSpring=true
    bGravityOverrideInSimSpace=true
    bOverrideAngularDamping=true
    gravityScale=0.35
    simulationSpace=RootRelative
    alphaBoolBlend {blendOption=Linear}
    boundBone {boneName=dyn_skirtsideB_01_r}
    chainEnd {boneName=dyn_skirtsideB_02_r}
    componentAppliedLinearAccClamp=<500 500 150>
    componentLinearAccScale=<25 25 25>
    componentLinearVelScale=<25 25 25>
    physicsBodyDefinitions [{
      sphereCollisionRadius=0
      boundBone {boneName=dyn_skirtsideB_01_r}
      boxExtents=<6 12 2>
      constraintSetup {angularLimitsMax=<10 25 5>, angularLimitsMin=<-10 0 -5>, angularTarget=<1 0 0>}
      localJointOffset=<1 0 0>
     }]
    relativeSpaceBone {boneName=pelvis}
   }
   n13 = Anim Dynamics - Bone: dyn_skirtsideB_02_r(ComponentPose=@n12, Alpha=1, ExternalForce="0,0,0") {
    angularDampingOverride=0.9999
    angularSpringConstant=50
    bAngularSpring=true
    bGravityOverrideInSimSpace=true
    bOverrideAngularDamping=true
    gravityScale=0.1
    linearSpringConstant=50
    simulationSpace=RootRelative
    alphaBoolBlend {blendOption=Linear}
    boundBone {boneName=dyn_skirtsideB_02_r}
    chainEnd {boneName=dyn_skirtsideB_03_r}
    componentAppliedLinearAccClamp=<500 500 150>
    componentLinearAccScale=<25 25 25>
    componentLinearVelScale=<25 25 25>
    physicsBodyDefinitions [{
      sphereCollisionRadius=0
      boundBone {boneName=dyn_skirtsideB_02_r}
      boxExtents=<6 12 2>
      constraintSetup {angularLimitsMax=<0 25 15>, angularLimitsMin=<0 0 0>, angularTarget=<1 0 0>}
      localJointOffset=<1 0 0>
     }]
    relativeSpaceBone {boneName=pelvis}
   }
   n14 = Anim Dynamics - Bone: dyn_skirtsideB_03_r(ComponentPose=@n13, Alpha=1, ExternalForce="0,0,0") {
    angularDampingOverride=0.9999
    angularSpringConstant=50
    bAngularSpring=true
    bGravityOverrideInSimSpace=true
    bOverrideAngularDamping=true
    gravityScale=0.1
    linearSpringConstant=50
    simulationSpace=RootRelative
    alphaBoolBlend {blendOption=Linear}
    boundBone {boneName=dyn_skirtsideB_03_r}
    chainEnd {boneName=dyn_skirtsideB_04_r}
    componentAppliedLinearAccClamp=<500 500 150>
    componentLinearAccScale=<25 25 25>
    componentLinearVelScale=<25 25 25>
    physicsBodyDefinitions [{
      sphereCollisionRadius=0
      boundBone {boneName=dyn_skirtsideB_03_r}
      boxExtents=<6 12 2>
      constraintSetup {angularLimitsMax=<25 50 35>, angularLimitsMin=<0 0 0>, angularTarget=<1 0 0>}
      localJointOffset=<1 0 0>
     }]
    relativeSpaceBone {boneName=pelvis}
   }
   Anim Dynamics - Bone: dyn_skirtsideB_04_r(ComponentPose=@n14, Alpha=1, ExternalForce="0,0,0") {
    angularDampingOverride=0.9999
    angularSpringConstant=50
    bAngularSpring=true
    bGravityOverrideInSimSpace=true
    bOverrideAngularDamping=true
    gravityScale=0.4
    linearSpringConstant=50
    simulationSpace=RootRelative
    alphaBoolBlend {blendOption=Linear}
    boundBone {boneName=dyn_skirtsideB_04_r}
    chainEnd {boneName=dyn_skirtsideB_04_r}
    componentAppliedLinearAccClamp=<500 500 150>
    componentLinearAccScale=<25 25 25>
    componentLinearVelScale=<25 25 25>
    physicsBodyDefinitions [{
      sphereCollisionRadius=0
      boundBone {boneName=dyn_skirtsideB_04_r}
      boxExtents=<6 12 2>
      constraintSetup {angularLimitsMax=<90 90 35>, angularLimitsMin=<0 0 0>, angularTarget=<1 0 0>}
      localJointOffset=<0 1 0>
     }]
    relativeSpaceBone {boneName=pelvis}
   }
   n15 = Anim Dynamics - Bone: dyn_skirtsideB_00_l(Alpha=1, ExternalForce="0,0,0") {
    angularDampingOverride=0.9999
    angularSpringConstant=50
    bAngularSpring=true
    bGravityOverrideInSimSpace=true
    bOverrideAngularDamping=true
    gravityScale=0.1
    simulationSpace=RootRelative
    alphaBoolBlend {blendOption=Linear}
    boundBone {boneName=dyn_skirtsideB_00_l}
    componentAppliedLinearAccClamp=<500 500 500>
    componentLinearAccScale=<25 25 25>
    componentLinearVelScale=<25 25 25>
    physicsBodyDefinitions [{
      sphereCollisionRadius=0
      boundBone {boneName=dyn_skirtsideB_00_l}
      boxExtents=<6 12 2>
      constraintSetup {angularLimitsMax=<10 0 5>, angularLimitsMin=<-10 -35 -15>, angularTarget=<1 0 0>}
      localJointOffset=<1 0 0>
     }]
    relativeSpaceBone {boneName=pelvis}
   }
   n16 = Anim Dynamics - Bone: dyn_skirtsideB_01_l(ComponentPose=@n15, Alpha=1, ExternalForce="0,0,0") {
    angularDampingOverride=0.9999
    angularSpringConstant=50
    bAngularSpring=true
    bGravityOverrideInSimSpace=true
    bOverrideAngularDamping=true
    gravityScale=0.35
    simulationSpace=RootRelative
    alphaBoolBlend {blendOption=Linear}
    boundBone {boneName=dyn_skirtsideB_01_l}
    componentAppliedLinearAccClamp=<500 500 150>
    componentLinearAccScale=<25 25 25>
    componentLinearVelScale=<25 25 25>
    physicsBodyDefinitions [{
      sphereCollisionRadius=0
      boundBone {boneName=dyn_skirtsideB_01_l}
      boxExtents=<6 12 2>
      constraintSetup {angularLimitsMax=<10 0 5>, angularLimitsMin=<-10 -50 -5>, angularTarget=<1 0 0>}
      localJointOffset=<1 0 0>
     }]
    relativeSpaceBone {boneName=pelvis}
   }
   n17 = Anim Dynamics - Bone: dyn_skirtsideB_02_l(ComponentPose=@n16, Alpha=1, ExternalForce="0,0,0") {
    angularDampingOverride=0.9999
    angularSpringConstant=50
    bAngularSpring=true
    bGravityOverrideInSimSpace=true
    bOverrideAngularDamping=true
    gravityScale=0.35
    simulationSpace=RootRelative
    alphaBoolBlend {blendOption=Linear}
    boundBone {boneName=dyn_skirtsideB_02_l}
    chainEnd {boneName=dyn_skirtsideB_03_l}
    componentAppliedLinearAccClamp=<500 500 150>
    componentLinearAccScale=<25 25 25>
    componentLinearVelScale=<25 25 25>
    physicsBodyDefinitions [{
      sphereCollisionRadius=0
      boundBone {boneName=dyn_skirtsideB_02_l}
      boxExtents=<6 12 2>
      constraintSetup {angularLimitsMax=<0 0 15>, angularLimitsMin=<0 -50 0>, angularTarget=<1 0 0>}
      localJointOffset=<1 0 0>
     }]
    relativeSpaceBone {boneName=pelvis}
   }
   n18 = Anim Dynamics - Bone: dyn_skirtsideB_03_l(ComponentPose=@n17, Alpha=1, ExternalForce="0,0,0") {
    angularDampingOverride=0.9999
    angularSpringConstant=50
    bAngularSpring=true
    bGravityOverrideInSimSpace=true
    bOverrideAngularDamping=true
    gravityScale=0.35
    simulationSpace=RootRelative
    alphaBoolBlend {blendOption=Linear}
    boundBone {boneName=dyn_skirtsideB_03_l}
    componentAppliedLinearAccClamp=<500 500 150>
    componentLinearAccScale=<25 25 25>
    componentLinearVelScale=<25 25 25>
    physicsBodyDefinitions [{
      sphereCollisionRadius=0
      boundBone {boneName=dyn_skirtsideB_03_l}
      boxExtents=<6 12 2>
      constraintSetup {angularLimitsMax=<0 0 35>, angularLimitsMin=<0 -25 0>, angularTarget=<1 0 0>}
      localJointOffset=<1 0 0>
     }]
    relativeSpaceBone {boneName=pelvis}
   }
   Anim Dynamics - Bone: dyn_skirtsideB_04_l(ComponentPose=@n18, Alpha=1, ExternalForce="0,0,0") {
    angularDampingOverride=0.9999
    angularSpringConstant=50
    bAngularSpring=true
    bGravityOverrideInSimSpace=true
    bOverrideAngularDamping=true
    gravityScale=0.4
    simulationSpace=RootRelative
    alphaBoolBlend {blendOption=Linear}
    boundBone {boneName=dyn_skirtsideB_04_l}
    componentAppliedLinearAccClamp=<500 500 150>
    componentLinearAccScale=<25 25 25>
    componentLinearVelScale=<25 25 25>
    physicsBodyDefinitions [{
      sphereCollisionRadius=0
      boundBone {boneName=dyn_skirtsideB_04_l}
      boxExtents=<6 12 2>
      constraintSetup {angularLimitsMax=<0 0 35>, angularLimitsMin=<-25 -25 0>, angularTarget=<1 0 0>}
      localJointOffset=<0 1 0>
     }]
    relativeSpaceBone {boneName=pelvis}
   }
   n19 = Get is_crouching_and_not_walking_alpha() -> (is_crouching_and_not_walking_alpha:double)
   n20 = "Transform (Modify) Bone - Bone: dyn_skirtsideB_00_l"(Alpha=@n19, Translation="0,0,0", Rotation="25,5,35", Scale="1,1,1") {
    rotationMode=BMM_Additive
    rotationSpace=BCS_ParentBoneSpace
    translationMode=BMM_Additive
    translationSpace=BCS_ParentBoneSpace
    alphaBoolBlend {blendOption=Linear}
    boneToModify {boneName=dyn_skirtsideB_00_l}
    rotation=<rot 25 5 35>
   }
   n21 = "Transform (Modify) Bone - Bone: dyn_skirtsideB_01_l"(ComponentPose=@n20, Alpha=@n19, Translation="0,0,0", Rotation="-45,0,-35", Scale="1,1,1") {
    rotationMode=BMM_Additive
    rotationSpace=BCS_ParentBoneSpace
    translationMode=BMM_Additive
    translationSpace=BCS_ParentBoneSpace
    alphaBoolBlend {blendOption=Linear}
    boneToModify {boneName=dyn_skirtsideB_01_l}
    rotation {pitch=-45, roll=-35}
   }
   n22 = "Transform (Modify) Bone - Bone: dyn_skirtsideB_02_l"(ComponentPose=@n21, Alpha=@n19, Translation="0,0,0", Rotation="-60,0,0", Scale="1,1,1") {
    rotationMode=BMM_Additive
    rotationSpace=BCS_ParentBoneSpace
    translationMode=BMM_Additive
    translationSpace=BCS_ParentBoneSpace
    alphaBoolBlend {blendOption=Linear}
    boneToModify {boneName=dyn_skirtsideB_02_l}
    rotation {pitch=-60}
   }
   n23 = "Transform (Modify) Bone - Bone: dyn_skirtsideB_03_r"(ComponentPose=@n22, Alpha=@n19, Translation="0,0,0", Rotation="-30,0,0", Scale="1,1,1") {
    rotationMode=BMM_Additive
    rotationSpace=BCS_ParentBoneSpace
    translationMode=BMM_Additive
    translationSpace=BCS_ParentBoneSpace
    alphaBoolBlend {blendOption=Linear}
    boneToModify {boneName=dyn_skirtsideB_03_r}
    rotation {pitch=-30}
   }
   n24 = "Transform (Modify) Bone - Bone: dyn_skirtsideB_00_l"(ComponentPose=@n23, Alpha=1, Translation="0,0,0", Rotation="11,0,0", Scale="1,1,1") {
    rotationMode=BMM_Additive
    rotationSpace=BCS_ParentBoneSpace
    translationMode=BMM_Additive
    translationSpace=BCS_ParentBoneSpace
    alphaBoolBlend {blendOption=Linear}
    boneToModify {boneName=dyn_skirtsideB_00_l}
    rotation {pitch=11}
   }
   "Transform (Modify) Bone - Bone: dyn_skirtsideB_00_r"(ComponentPose=@n24, Alpha=1, Translation="0,0,0", Rotation="-11,0,0", Scale="1,1,1") {
    rotationMode=BMM_Additive
    rotationSpace=BCS_ParentBoneSpace
    translationMode=BMM_Additive
    translationSpace=BCS_ParentBoneSpace
    alphaBoolBlend {blendOption=Linear}
    boneToModify {boneName=dyn_skirtsideB_00_r}
    rotation {pitch=-11}
   }
   "Transform (Modify) Bone - Bone: dyn_skirtsideB_03_r"(Alpha=Get is_crouching_and_not_walking_alpha(), Translation="0,0,0", Rotation="0,35,0", Scale="1,1,1") {rotationMode=BMM_Additive, alphaBoolBlend {blendOption=Linear}, boneToModify {boneName=dyn_skirtsideB_03_r}, rotation {yaw=35}}
   n25 = RigidBody(ComponentPose=Local To Component(LocalPose=Use cached pose 'CollarCache'()), Alpha=Get loinClothAlpha() -> (loinClothAlpha:double)) {
    bClampLinearTranslationLimitToRefPose=true
    bForceDisableCollisionBetweenConstraintBodies=true
    bOverrideWorldGravity=true
    overridePhysicsAsset=&/F_MED_Butterfly_FrontLoincloth_Dyn_Loincloth_Front
    simulationSpace=BaseBoneSpace
    alphaBoolBlend {blendOption=Linear}
    baseBoneRef {boneName=root}
    componentAppliedLinearAccClamp=<1000 1000 1000>
    componentLinearAccScale=<7 7 7>
    componentLinearVelScale=<7 7 7>
    overrideWorldGravity {z=-980}
   }
   n26 = "Transform (Modify) Bone - Bone: dyn_skirt_frontbase_01"(ComponentPose=@n25, Alpha=Get crouchAlpha() -> (crouchAlpha:double), Translation="0,2,0", Rotation="0,0,0", Scale="1,1,1") {
    rotationMode=BMM_Additive
    rotationSpace=BCS_ParentBoneSpace
    translationMode=BMM_Additive
    translationSpace=BCS_ParentBoneSpace
    alphaBoolBlend {blendOption=Linear}
    boneToModify {boneName=dyn_skirt_frontbase_01}
    translation {y=2}
   }
   n27 = Get loinClothAlphaInv() -> (loinClothAlphaInv:double)
   n28 = Anim Dynamics - Bone: dyn_skirt_front_01(ComponentPose=@n26, Alpha=@n27, ExternalForce="0,0,0") {
    angularDampingOverride=0.9999
    angularSpringConstant=50
    bAngularSpring=true
    bGravityOverrideInSimSpace=true
    bOverrideAngularDamping=true
    gravityScale=0.35
    alphaBoolBlend {blendOption=Linear}
    boundBone {boneName=dyn_skirt_front_01}
    componentAppliedLinearAccClamp=<500 500 500>
    physicsBodyDefinitions [{
      sphereCollisionRadius=0
      boundBone {boneName=dyn_skirt_front_01}
      boxExtents=<6 2 12>
      constraintSetup {angularLimitsMax=<0 0 35>, angularLimitsMin=<0 0 -35>, angularTarget=<1 0 0>}
      localJointOffset=<1 0 0>
     }]
   }
   n29 = Anim Dynamics - Bone: dyn_skirt_front_02(ComponentPose=@n28, Alpha=@n27, ExternalForce="0,0,0") {
    angularDampingOverride=0.9999
    angularSpringConstant=50
    bAngularSpring=true
    bGravityOverrideInSimSpace=true
    bOverrideAngularDamping=true
    gravityScale=0.35
    alphaBoolBlend {blendOption=Linear}
    boundBone {boneName=dyn_skirt_front_02}
    componentAppliedLinearAccClamp=<500 500 500>
    physicsBodyDefinitions [{
      sphereCollisionRadius=0
      boundBone {boneName=dyn_skirt_front_02}
      boxExtents=<6 2 12>
      constraintSetup {angularLimitsMax=<0 0 35>, angularLimitsMin=<0 0 -35>, angularTarget=<1 0 0>}
      localJointOffset=<1 0 0>
     }]
   }
   n30 = Anim Dynamics - Bone: dyn_skirt_front_03(ComponentPose=@n29, Alpha=1, ExternalForce="0,0,0") {
    angularDampingOverride=0.9999
    angularSpringConstant=50
    bAngularSpring=true
    bGravityOverrideInSimSpace=true
    bOverrideAngularDamping=true
    gravityScale=0.35
    alphaBoolBlend {blendOption=Linear}
    boundBone {boneName=dyn_skirt_front_03}
    componentAppliedLinearAccClamp=<500 500 500>
    physicsBodyDefinitions [{
      sphereCollisionRadius=0
      boundBone {boneName=dyn_skirt_front_03}
      boxExtents=<6 2 12>
      constraintSetup {angularLimitsMax=<0 0 35>, angularLimitsMin=<0 0 -35>, angularTarget=<1 0 0>}
      localJointOffset=<1 0 0>
     }]
   }
   n31 = "Transform (Modify) Bone - Bone: dyn_skirt_front_01_r"(ComponentPose=@n30, Alpha=@n27, Translation="0,0,0", Rotation="0,45,0", Scale="1,1,1") {
    rotationMode=BMM_Additive
    rotationSpace=BCS_ParentBoneSpace
    translationMode=BMM_Additive
    translationSpace=BCS_ParentBoneSpace
    alphaBoolBlend {blendOption=Linear}
    boneToModify {boneName=dyn_skirt_front_01_r}
    rotation {yaw=45}
   }
   n32 = "Transform (Modify) Bone - Bone: dyn_skirt_front_02_r"(ComponentPose=@n31, Alpha=@n27, Translation="0,0,0", Rotation="0,45,0", Scale="1,1,1") {
    rotationMode=BMM_Additive
    rotationSpace=BCS_ParentBoneSpace
    translationMode=BMM_Additive
    translationSpace=BCS_ParentBoneSpace
    alphaBoolBlend {blendOption=Linear}
    boneToModify {boneName=dyn_skirt_front_02_r}
    rotation {yaw=45}
   }
   n33 = Get is_crouching_and_walking_alpha() -> (is_crouching_and_walking_alpha:double)
   n34 = "Transform (Modify) Bone - Bone: dyn_skirt_frontbase_01"(ComponentPose=@n32, Alpha=@n33, Translation="0,0,0", Rotation="0,-20,0", Scale="1,1,1") {
    rotationMode=BMM_Additive
    rotationSpace=BCS_ParentBoneSpace
    translationMode=BMM_Additive
    translationSpace=BCS_ParentBoneSpace
    alphaBoolBlend {blendOption=Linear}
    boneToModify {boneName=dyn_skirt_frontbase_01}
    rotation {yaw=-20}
   }
   n35 = "Transform (Modify) Bone - Bone: dyn_skirt_front_00"(ComponentPose=@n34, Alpha=@n33, Translation="0,0,0", Rotation="0,0,-45", Scale="1,1,1") {
    rotationMode=BMM_Additive
    rotationSpace=BCS_ParentBoneSpace
    translationMode=BMM_Additive
    translationSpace=BCS_ParentBoneSpace
    alphaBoolBlend {blendOption=Linear}
    boneToModify {boneName=dyn_skirt_front_00}
    rotation {roll=-45}
   }
   n36 = RigidBody(ComponentPose=@n35, Alpha=0.9) {
    bClampLinearTranslationLimitToRefPose=true
    bOverrideWorldGravity=true
    overridePhysicsAsset=&/F_MED_Butterfly_BackSkirt_Dyn_Skirt
    simulationSpace=BaseBoneSpace
    alphaBoolBlend {blendOption=Linear}
    baseBoneRef {boneName=spine_01}
    componentAppliedLinearAccClamp=<3000 3000 3000>
    componentLinearAccScale=<0.4 0.4 0.4>
    componentLinearVelScale=<0.4 0.4 0.4>
    overrideWorldGravity {z=-980}
   }
   n37 = "Transform (Modify) Bone - Bone: dyn_skirtback_base"(
    ComponentPose=@n36
    Alpha=Get Is_TacticalSprinting() -> (Is_TacticalSprinting:double)
    Translation="-0.052,-0.006,0.564"
    Rotation="7.5,0,0"
    Scale="1,1,1"
   ) {
    rotationMode=BMM_Additive
    rotationSpace=BCS_BoneSpace
    translationMode=BMM_Additive
    translationSpace=BCS_BoneSpace
    alphaBoolBlend {blendOption=Linear}
    boneToModify {boneName=dyn_skirtback_base}
    rotation {pitch=7.5}
    translation=<-0.052 -0.006 0.564>
   }
   n38 = "Transform (Modify) Bone - Bone: dyn_skirtback_base_l"(ComponentPose=@n37, Alpha=Get Is_TacticalSprinting(), Translation="-0.061,0.010,0.607", Rotation="6.5,0,0", Scale="1,1,1") {
    rotationMode=BMM_Additive
    rotationSpace=BCS_BoneSpace
    translationMode=BMM_Additive
    translationSpace=BCS_BoneSpace
    alphaBoolBlend {blendOption=Linear}
    boneToModify {boneName=dyn_skirtback_base_l}
    rotation {pitch=6.5}
    translation=<-0.061 0.01 0.607>
   }
   n39 = "Transform (Modify) Bone - Bone: dyn_skirtback_base_r"(ComponentPose=@n38, Alpha=Get Is_TacticalSprinting(), Translation="-0.078,-0.034,0.785", Rotation="6.498,-2.021,-0.192", Scale="1,1,1") {
    rotationMode=BMM_Additive
    rotationSpace=BCS_BoneSpace
    translationMode=BMM_Additive
    translationSpace=BCS_BoneSpace
    alphaBoolBlend {blendOption=Linear}
    boneToModify {boneName=dyn_skirtback_base_r}
    rotation=<rot 6.4979 -2.021 -0.192>
    translation=<-0.078 -0.034 0.785>
   }
   n40 = "Transform (Modify) Bone - Bone: dyn_skirtback_base"(ComponentPose=@n39, Alpha=1, Translation="0,0,0", Rotation="-7.447,-0.017,-0.001", Scale="1,1,1") {
    rotationMode=BMM_Additive
    rotationSpace=BCS_BoneSpace
    alphaBoolBlend {blendOption=Linear}
    boneToModify {boneName=dyn_skirtback_base}
    rotation=<rot -7.447 -0.017 -0.001>
   }
   n41 = "Transform (Modify) Bone - Bone: dyn_skirtback_base_r"(ComponentPose=@n40, Alpha=1, Translation="0,0,0", Rotation="-8.707,-0.125,0.122", Scale="1,1,1") {
    rotationMode=BMM_Additive
    rotationSpace=BCS_BoneSpace
    alphaBoolBlend {blendOption=Linear}
    boneToModify {boneName=dyn_skirtback_base_r}
    rotation=<rot -8.707 -0.125 0.122>
   }
   n42 = "Transform (Modify) Bone - Bone: dyn_skirtback_base_l"(ComponentPose=@n41, Alpha=1, Translation="0,0,0", Rotation="-10.508,-0.069,-0.014", Scale="1,1,1") {
    rotationMode=BMM_Additive
    rotationSpace=BCS_BoneSpace
    alphaBoolBlend {blendOption=Linear}
    boneToModify {boneName=dyn_skirtback_base_l}
    rotation=<rot -10.508 -0.069 -0.014>
   }
   n43 = "Constraint - dyn_skirt_front_00 (RR)"!(ComponentPose=Local To Component(LocalPose=Use cached pose 'CollarCache'()), Alpha=1, ConstraintWeights_0=0.5, ConstraintWeights_1:float=0.5) -> (Pose) {
    alphaBoolBlend {blendOption=Linear}
    boneToModify {boneName=dyn_skirt_front_00}
    constraintSetup [{transformType=Rotation, targetBone {boneName=thigh_l}}, {transformType=Rotation, targetBone {boneName=thigh_r}}]
    constraintWeights [1,1]
   }
   n44 = Get crouchAlpha()
   n45 = Copy Bone Delta - Source Bone: dyn_skirt_front_00 - Target Bone: dyn_skirt_front_01!(
    ComponentPose:Component Space Pose Link Structure=@n43
    Alpha:float=@n44
    bCopyTranslation:bool=False
    bCopyRotation:bool=true
    bCopyScale:bool=False
   ) -> (Pose:Component Space Pose Link Structure) {
    is=title AnimGraphNode_CopyBoneDelta
    rotationMultiplier=-0.5
    alphaBoolBlend {blendOption=Linear}
    sourceBone {boneName=dyn_skirt_front_00}
    targetBone {boneName=dyn_skirt_front_01}
   }
   n46 = "Transform (Modify) Bone - Bone: dyn_skirt_front_00"(ComponentPose=@n45, Alpha=@n44, Translation="0,0,0", Rotation="0,-25,0", Scale="1,1,1") {
    rotationMode=BMM_Additive
    rotationSpace=BCS_BoneSpace
    alphaBoolBlend {blendOption=Linear}
    boneToModify {boneName=dyn_skirt_front_00}
    rotation {yaw=-25}
   }
   n47 = "Transform (Modify) Bone - Bone: dyn_skirtsideB_00_l"(ComponentPose=@n46, Alpha=1, Translation="0,0,0", Rotation="15,0,0", Scale="1,1,1") {
    rotationMode=BMM_Additive
    rotationSpace=BCS_BoneSpace
    alphaBoolBlend {blendOption=Linear}
    boneToModify {boneName=dyn_skirtsideB_00_l}
    rotation {pitch=15}
   }
   n48 = "Transform (Modify) Bone - Bone: dyn_skirtsideB_00_r"(ComponentPose=@n47, Alpha=1, Translation="0,0,0", Rotation="-15,0,0", Scale="1,1,1") {
    rotationMode=BMM_Additive
    rotationSpace=BCS_BoneSpace
    alphaBoolBlend {blendOption=Linear}
    boneToModify {boneName=dyn_skirtsideB_00_r}
    rotation {pitch=-15}
   }
   n49 = Blend Poses by bool(
    BlendPose_0=Component To Local(ComponentPose=@n48)
    BlendPose_1=Component To Local(ComponentPose=@n42)
    bActiveValue=Get rigidsOnOffBool() -> (rigidsOnOffBool:bool)
   ) {blendType=Linear}
   Output Pose(Result=Component To Local(ComponentPose=Local To Component(LocalPose=@n49))) {layerGroup=None}
   n50 = Anim Dynamics - Bone: dyn_arm_l(ComponentPose=Local To Component(LocalPose=Copy Pose From Mesh() {bUseAttachedParent=true}), Alpha=0.75, ExternalForce="0,0,0") {
    angularDampingOverride=1
    angularSpringConstant=3500
    bAngularSpring=true
    bGravityOverrideInSimSpace=true
    bOverrideAngularDamping=true
    bUseGravityOverride=true
    gravityScale=0.35
    simulationSpace=BoneRelative
    alphaBoolBlend {blendOption=Linear}
    boundBone {boneName=dyn_arm_l}
    componentAppliedLinearAccClamp=<3000 3000 3000>
    componentLinearAccScale=<0.2 0.2 0.2>
    componentLinearVelScale=<0.2 0.2 0.2>
    gravityOverride {z=-980}
    physicsBodyDefinitions [{
      sphereCollisionRadius=0
      boundBone {boneName=dyn_arm_l}
      boxExtents=<10 10 10>
      constraintSetup {angularLimitsMax=<5 15 5>, angularLimitsMin=<-5 -15 -5>, angularTarget=<1 0 0>}
      localJointOffset=<1 0 0>
     }]
    relativeSpaceBone {boneName=spine_05}
   }
   n51 = Anim Dynamics - Bone: dyn_arm_r(ComponentPose=@n50, Alpha=0.75, ExternalForce="0,0,0") {
    angularDampingOverride=1
    angularSpringConstant=3500
    bAngularSpring=true
    bGravityOverrideInSimSpace=true
    bOverrideAngularDamping=true
    bUseGravityOverride=true
    gravityScale=0.35
    simulationSpace=BoneRelative
    alphaBoolBlend {blendOption=Linear}
    boundBone {boneName=dyn_arm_r}
    componentAppliedLinearAccClamp=<500 500 500>
    gravityOverride {z=-980}
    physicsBodyDefinitions [{
      sphereCollisionRadius=0
      boundBone {boneName=dyn_arm_r}
      boxExtents=<10 10 10>
      constraintSetup {angularLimitsMax=<5 15 5>, angularLimitsMin=<-5 -15 -5>, angularTarget=<1 0 0>}
      localJointOffset=<-1 0 0>
     }]
    relativeSpaceBone {boneName=spine_05}
   }
   n52 = Get neckTargetingRemapAlpha() -> (neckTargetingRemapAlpha:double)
   n53 = Anim Dynamics - Bone: dyn_collarSide_01_l(ComponentPose=@n51, Alpha=@n52, ExternalForce="0,0,0") {
    angularDampingOverride=0.9999
    angularSpringConstant=200
    bAngularSpring=true
    bGravityOverrideInSimSpace=true
    bOverrideAngularDamping=true
    gravityScale=0.1
    alphaBoolBlend {blendOption=Linear}
    boundBone {boneName=dyn_collarSide_01_l}
    componentAppliedLinearAccClamp=<500 500 500>
    componentLinearAccScale=<0.5 0.5 0.5>
    componentLinearVelScale=<0.5 0.5 0.5>
    physicsBodyDefinitions [{
      sphereCollisionRadius=0
      boundBone {boneName=dyn_collarSide_01_l}
      boxExtents=<3 3 3>
      constraintSetup {angularLimitsMax=<2 0 2>, angularLimitsMin=<-2 0 -2>, angularTarget=<1 0 0>}
      localJointOffset=<0.5 0 0>
     }]
    relativeSpaceBone {boneName=neck_01}
   }
   n54 = Anim Dynamics - Bone: dyn_collarSide_02_l(ComponentPose=@n53, Alpha=@n52, ExternalForce="0,0,0") {
    angularDampingOverride=0.9999
    angularSpringConstant=200
    bAngularSpring=true
    bGravityOverrideInSimSpace=true
    bOverrideAngularDamping=true
    gravityScale=0.1
    alphaBoolBlend {blendOption=Linear}
    boundBone {boneName=dyn_collarSide_02_l}
    componentAppliedLinearAccClamp=<500 500 500>
    componentLinearAccScale=<0.5 0.5 0.5>
    componentLinearVelScale=<0.5 0.5 0.5>
    physicsBodyDefinitions [{
      sphereCollisionRadius=0
      boundBone {boneName=dyn_collarSide_02_l}
      boxExtents=<3 3 3>
      constraintSetup {angularLimitsMax=<5 0 5>, angularLimitsMin=<0 0 -5>, angularTarget=<1 0 0>}
      localJointOffset=<0.5 0 0>
     }]
    relativeSpaceBone {boneName=neck_01}
   }
   n55 = Anim Dynamics - Bone: dyn_collarSide_01_r(ComponentPose=@n54, Alpha=@n52, ExternalForce="0,0,0") {
    angularDampingOverride=0.9999
    angularSpringConstant=200
    bAngularSpring=true
    bGravityOverrideInSimSpace=true
    bOverrideAngularDamping=true
    gravityScale=0.1
    alphaBoolBlend {blendOption=Linear}
    boundBone {boneName=dyn_collarSide_01_r}
    componentAppliedLinearAccClamp=<500 500 500>
    componentLinearAccScale=<0.5 0.5 0.5>
    componentLinearVelScale=<0.5 0.5 0.5>
    physicsBodyDefinitions [{
      sphereCollisionRadius=0
      boundBone {boneName=dyn_collarSide_01_r}
      boxExtents=<3 3 3>
      constraintSetup {angularLimitsMax=<2 0 2>, angularLimitsMin=<-2 0 -2>, angularTarget=<1 0 0>}
      localJointOffset=<0.5 0 0>
     }]
    relativeSpaceBone {boneName=neck_01}
   }
   n56 = Anim Dynamics - Bone: dyn_collarSide_02_r(ComponentPose=@n55, Alpha=@n52, ExternalForce="0,0,0") {
    angularDampingOverride=0.9999
    angularSpringConstant=200
    bAngularSpring=true
    bGravityOverrideInSimSpace=true
    bOverrideAngularDamping=true
    gravityScale=0.1
    alphaBoolBlend {blendOption=Linear}
    boundBone {boneName=dyn_collarSide_02_r}
    componentAppliedLinearAccClamp=<500 500 500>
    componentLinearAccScale=<0.5 0.5 0.5>
    componentLinearVelScale=<0.5 0.5 0.5>
    physicsBodyDefinitions [{
      sphereCollisionRadius=0
      boundBone {boneName=dyn_collarSide_02_r}
      boxExtents=<3 3 3>
      constraintSetup {angularLimitsMax=<0 0 5>, angularLimitsMin=<-5 0 -5>, angularTarget=<1 0 0>}
      localJointOffset=<0.5 0 0>
     }]
    relativeSpaceBone {boneName=neck_01}
   }
   n57 = Anim Dynamics - Bone: dyn_collarBack_01(ComponentPose=@n56, Alpha=@n52, ExternalForce="0,0,0") {
    angularDampingOverride=0.9999
    angularSpringConstant=200
    bAngularSpring=true
    bGravityOverrideInSimSpace=true
    bOverrideAngularDamping=true
    gravityScale=0.1
    alphaBoolBlend {blendOption=Linear}
    boundBone {boneName=dyn_collarBack_01}
    componentAppliedLinearAccClamp=<500 500 500>
    componentLinearAccScale=<0.5 0.5 0.5>
    componentLinearVelScale=<0.5 0.5 0.5>
    physicsBodyDefinitions [{
      sphereCollisionRadius=0
      boundBone {boneName=dyn_collarBack_01}
      boxExtents=<3 3 3>
      constraintSetup {angularLimitsMax=<2 0 0>, angularLimitsMin=<-2 0 0>, angularTarget=<1 0 0>}
      localJointOffset=<0.5 0 0>
     }]
    relativeSpaceBone {boneName=neck_01}
   }
   n58 = Anim Dynamics - Bone: dyn_collarBack_02(ComponentPose=@n57, Alpha=@n52, ExternalForce="0,0,0") {
    angularDampingOverride=0.9999
    angularSpringConstant=200
    bAngularSpring=true
    bGravityOverrideInSimSpace=true
    bOverrideAngularDamping=true
    gravityScale=0.1
    alphaBoolBlend {blendOption=Linear}
    boundBone {boneName=dyn_collarBack_02}
    componentAppliedLinearAccClamp=<500 500 500>
    componentLinearAccScale=<0.5 0.5 0.5>
    componentLinearVelScale=<0.5 0.5 0.5>
    physicsBodyDefinitions [{
      sphereCollisionRadius=0
      boundBone {boneName=dyn_collarBack_02}
      boxExtents=<3 3 3>
      constraintSetup {angularLimitsMax=<5 0 5>, angularLimitsMin=<-5 0 -5>, angularTarget=<1 0 0>}
      localJointOffset=<0.5 0 0>
     }]
    relativeSpaceBone {boneName=neck_01}
   }
   n59 = "Transform (Modify) Bone - Bone: dyn_collarSide_01_l"(ComponentPose=@n58, Alpha=1, Translation="0,0,0", Rotation="-7,0,0", Scale="1,1,1") {
    rotationMode=BMM_Additive
    rotationSpace=BCS_ParentBoneSpace
    translationMode=BMM_Additive
    translationSpace=BCS_ParentBoneSpace
    alphaBoolBlend {blendOption=Linear}
    boneToModify {boneName=dyn_collarSide_01_l}
    rotation {pitch=-7}
   }
   n60 = "Transform (Modify) Bone - Bone: dyn_collarBack_01"(ComponentPose=@n59, Alpha=1, Translation="0,0,0", Rotation="0,-7,0", Scale="1,1,1") {
    rotationMode=BMM_Additive
    rotationSpace=BCS_ParentBoneSpace
    translationMode=BMM_Additive
    translationSpace=BCS_ParentBoneSpace
    alphaBoolBlend {blendOption=Linear}
    boneToModify {boneName=dyn_collarBack_01}
    rotation {yaw=-7}
   }
   n61 = "Transform (Modify) Bone - Bone: dyn_collarSide_01_r"(ComponentPose=@n60, Alpha=1, Translation="0,0,0", Rotation="7,0,0.000", Scale="1,1,1") {
    rotationMode=BMM_Additive
    rotationSpace=BCS_ParentBoneSpace
    translationMode=BMM_Additive
    translationSpace=BCS_ParentBoneSpace
    alphaBoolBlend {blendOption=Linear}
    boneToModify {boneName=dyn_collarSide_01_r}
    rotation {pitch=7}
   }
   Save cached pose 'CollarCache'(Pose=Component To Local(ComponentPose=@n61)) {cacheName=CollarCache}
  ]}
 EventGraph {nodes [
   Event OnEndCrouching()
   > ResetDynamics(InTeleportType=TeleportPhysics)
   > ResetDynamics(InTeleportType=ResetPhysics)
   n1 = Event BlueprintUpdateAnimation()
   > n2 = Sequence!(execute) -> (then_0, then_1, then_2:Exec, then_3:Exec, then_4:Exec)
   n3 = Get CurrentLOD()
   n4 = Get bRigidBodyEnabled()
   > Set rigidsOnOffBool(rigidsOnOffBool:bool=OR Boolean(A=NOT Boolean(A=@n4), B="integer > integer"(A=@n3, B=1) {is=title K2Node_CallFunction})) -> (Output_Get:bool)
   n5 = Branch(execute=@n2.then_3, Condition=Get bIsTacticalSprinting())
   > Set Is_TacticalSprinting(Is_TacticalSprinting:double=FInterpTo(Current=Get Is_TacticalSprinting(), Target=1, DeltaTime=@n1.DeltaTimeX, InterpSpeed=12)) -> (Output_Get:double)
   Set Is_TacticalSprinting(execute=@n5.else, Is_TacticalSprinting=FInterpTo(Current=Get Is_TacticalSprinting(), Target=0, DeltaTime=@n1.DeltaTimeX, InterpSpeed=12))
   n6 = Get bIsTargeting()
   Set neckTargetingRemapAlpha(
    execute=@n2.then_2
    neckTargetingRemapAlpha:double=MapRangeClamped(Value="To Float (Boolean)"(InBool=@n6), InRangeA=0, InRangeB=1, OutRangeA=0.3, OutRangeB=0)
   ) -> (Output_Get:double)
   n7 = Get bIsCrouching()
   n8 = "To Float (Boolean)"(InBool=NOT Boolean(A=@n7))
   Set crouchAlpha(execute=@n2.then_1, crouchAlpha:double=@n8) -> (Output_Get:double)
   n9 = "To Float (Boolean)"(InBool=@n7)
   > Set crouchAlpha(crouchAlpha=@n9)
   n10 = Get PawnSpeed2D()
   > Set is_crouching_and_not_walking_alpha(is_crouching_and_not_walking_alpha:double=MapRangeClamped(Value=@n10, InRangeA=0, InRangeB=50, OutRangeA=@n9, OutRangeB=0)) -> (Output_Get:double)
   > Set is_crouching_and_walking_alpha(is_crouching_and_walking_alpha:double=MapRangeClamped(Value=@n10, InRangeA=0, InRangeB=50, OutRangeB=@n9)) -> (Output_Get:double)
   n11 = MapRangeClamped(Value=@n8, InRangeA=0, InRangeB=1)
   > Set loinClothAlpha(loinClothAlpha:double=@n11) -> (Output_Get:double)
   > Set loinClothAlphaInv(loinClothAlphaInv:double=MapRangeClamped(Value=@n11, InRangeA=0, InRangeB=1, OutRangeA=1, OutRangeB=0)) -> (Output_Get:double)
   n12 = NOT Boolean(A=Get bIsPlayingEmote())
   n13 = Get bRigidBodyEnabled()
   n14 = Get bIsInFrontEnd()
   Set SkirtMove(execute=@n2.then_4, SkirtMove:double="To Float (Boolean)"(InBool=AND Boolean!(A=@n14, B=@n13, C:bool=@n12) -> (ReturnValue))) -> (Output_Get:double)
  ]}
]
variables [
 crouchAlpha {type=number, default=1}
 is_crouching_and_not_walking_alpha {type=number}
 is_crouching_and_walking_alpha {type=number}
 Is_TacticalSprinting {type=number}
 loinClothAlpha {type=number}
 loinClothAlphaInv {type=number}
 neckTargetingRemapAlpha {type=number}
 rigidsOnOffBool {type=boolean}
 SkirtMove {type=number}
]
