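# Physics asset for the F_MED_Butterfly female character's face accessory mesh, giving its left and right braids physics bodies for secondary motion.
# Each braid is a four-capsule chain with a short, thinner tip, constrained link by link up to a spherical head body (zero-length capsule).
# Root joints have the tightest angle limits and locked twist, while the lower two joints on each side lock one swing axis.
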PhysicsAsset /BRCosmetics/Characters/Player/Female/Medium/Bodies/F_MED_Butterfly/Meshes/Parts/F_MED_Butterfly_FaceAcc_Physics
bodies {
 dyn_braid_01_l [{shapeType=Capsule, length=2.5, radius=1.5517, center=<0.568 0.0627 0.0007>, rotation=<rot -1.2566 -91.3152 -87.8523>}]
 dyn_braid_01_r [{shapeType=Capsule, length=2.5, radius=1.5517, center=<-0.568 -0.0627 -0.0007>, rotation=<rot -1.2566 -91.3152 92.1477>}]
 dyn_braid_02_l [{shapeType=Capsule, length=2.5, radius=1.5517, center=<0.568 0.0627 0.0007>, rotation=<rot -1.2566 -91.3152 -87.8523>}]
 dyn_braid_02_r [{shapeType=Capsule, length=2.5, radius=1.5517, center=<-0.568 -0.0627 -0.0007>, rotation=<rot -1.2566 -91.3152 92.1477>}]
 dyn_braid_03_l [{shapeType=Capsule, length=2.8017, radius=1.25, center=<0.568 0.0627 0.0007>, rotation=<rot -1.2566 -91.3152 -87.8523>}]
 dyn_braid_03_r [{shapeType=Capsule, length=2.8017, radius=1.25, center=<-0.568 -0.0627 -0.0007>, rotation=<rot -1.2566 -91.3152 92.1477>}]
 dyn_braid_04_l [{shapeType=Capsule, length=0.25, radius=1, center=<0.4922 0.0826 0.087>, rotation=<rot 9.8109 96.0216 -78.7223>}]
 dyn_braid_04_r [{shapeType=Capsule, length=0.25, radius=1, center=<-0.4922 -0.0826 -0.087>, rotation=<rot 9.8109 96.0216 101.2778>}]
 head [{shapeType=Capsule, length=0, radius=5.6761, center=<4.7305 0.0001 0>}]
}
constraints [
 {bone1Name=dyn_braid_03_l, bone2Name=dyn_braid_02_l, swing1LimitDegrees=45, swing1Motion=Locked, swing2LimitDegrees=45, twistLimitDegrees=45}
 {bone1Name=dyn_braid_02_l, bone2Name=dyn_braid_01_l, swing1LimitDegrees=45, swing2LimitDegrees=45, twistLimitDegrees=45}
 {bone1Name=dyn_braid_01_l, bone2Name=head, swing1LimitDegrees=15, swing2LimitDegrees=15, twistLimitDegrees=15, twistMotion=Locked}
 {bone1Name=dyn_braid_04_l, bone2Name=dyn_braid_03_l, swing1LimitDegrees=30, swing1Motion=Locked, swing2LimitDegrees=30, twistLimitDegrees=30}
 {bone1Name=dyn_braid_01_r, bone2Name=head, swing1LimitDegrees=15, swing2LimitDegrees=15, twistLimitDegrees=15, twistMotion=Locked}
 {bone1Name=dyn_braid_03_r, bone2Name=dyn_braid_02_r, swing1LimitDegrees=45, swing1Motion=Locked, swing2LimitDegrees=45, twistLimitDegrees=45}
 {bone1Name=dyn_braid_02_r, bone2Name=dyn_braid_01_r, swing1LimitDegrees=45, swing2LimitDegrees=45, twistLimitDegrees=45}
 {bone1Name=dyn_braid_04_r, bone2Name=dyn_braid_03_r, swing1LimitDegrees=30, swing1Motion=Locked, swing2LimitDegrees=30, twistLimitDegrees=30}
]
