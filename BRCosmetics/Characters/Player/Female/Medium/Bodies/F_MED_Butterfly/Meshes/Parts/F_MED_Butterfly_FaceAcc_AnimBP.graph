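# Post-physics AnimBlueprint for the female Butterfly body's face accessory mesh, adding secondary motion to its dyn_newhair bones.
# It copies the attached parent's pose, runs a series of per-bone spring AnimDynamics nodes, then a head-based RigidBody node using its own physics asset.
# Targeting and jumping states set the blend alpha for most simulation nodes.
# Jumping also triggers a delayed upward AnimDynamics force, cleared when not jumping.
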AnimBlueprint /BRCosmetics/Characters/Player/Female/Medium/Bodies/F_MED_Butterfly/Meshes/Parts/F_MED_Butterfly_FaceAcc_AnimBP
target_skeleton=&/F_MED_Butterfly_FaceAcc_Skeleton
graphs [
 AnimGraph {nodes [
   n1 = Get jumpForce() -> (jumpForce:Vector)
   n2 = Anim Dynamics - Bone: dyn_newhair_01(ComponentPose=Local To Component(LocalPose=Copy Pose From Mesh() {bUseAttachedParent=true}), Alpha=1, ExternalForce=@n1) {
    angularDampingOverride=1
    angularSpringConstant=50
    bAngularSpring=true
    bGravityOverrideInSimSpace=true
    bOverrideAngularDamping=true
    gravityScale=0.35
    linearSpringConstant=50
    alphaBoolBlend {blendOption=Linear}
    boundBone {boneName=dyn_newhair_01}
    componentAppliedLinearAccClamp=<250 250 250>
    componentLinearAccScale=<1 1 5>
    componentLinearVelScale=<1 1 1>
    physicsBodyDefinitions [{
      sphereCollisionRadius=0
      boundBone {boneName=dyn_newhair_01}
      boxExtents=<5 5 5>
      constraintSetup {angularLimitsMax=<3 3 3>, angularLimitsMin=<-3 -3 -3>, angularTarget=<1 0 0>}
      localJointOffset=<2 0 0>
     }]
   }
   n3 = Anim Dynamics - Bone: dyn_newhair_07(ComponentPose=@n2, Alpha=1, ExternalForce=@n1) {
    angularDampingOverride=1
    angularSpringConstant=50
    bAngularSpring=true
    bGravityOverrideInSimSpace=true
    bOverrideAngularDamping=true
    gravityScale=0.35
    alphaBoolBlend {blendOption=Linear}
    boundBone {boneName=dyn_newhair_07}
    componentAppliedLinearAccClamp=<250 250 250>
    componentLinearAccScale=<1 1 5>
    componentLinearVelScale=<1 1 1>
    physicsBodyDefinitions [{
      sphereCollisionRadius=0
      boundBone {boneName=dyn_newhair_07}
      boxExtents=<5 5 5>
      constraintSetup {angularLimitsMax=<3 3 3>, angularLimitsMin=<-3 -3 -3>, angularTarget=<1 0 0>}
      localJointOffset=<2 0 0>
     }]
   }
   n4 = Get rigidDanglesAlpha() -> (rigidDanglesAlpha:double)
   n5 = Anim Dynamics - Bone: dyn_newhair_02(ComponentPose=@n3, Alpha=@n4, ExternalForce=@n1) {
    angularDampingOverride=1
    angularSpringConstant=50
    bAngularSpring=true
    bGravityOverrideInSimSpace=true
    bLinearSpring=true
    bOverrideAngularDamping=true
    gravityScale=0.35
    linearSpringConstant=50
    simulationSpace=RootRelative
    alphaBoolBlend {blendOption=Linear}
    boundBone {boneName=dyn_newhair_02}
    componentAppliedLinearAccClamp=<1000 400 400>
    componentLinearAccScale=<24 24 50>
    componentLinearVelScale=<24 24 50>
    physicsBodyDefinitions [{
      sphereCollisionRadius=0
      boundBone {boneName=dyn_newhair_02}
      boxExtents=<5 5 5>
      constraintSetup {angularLimitsMax=<20 20 20>, angularLimitsMin=<-20 -20 -20>, angularTarget=<1 0 0>}
      localJointOffset=<2 0 0>
     }]
   }
   n6 = Anim Dynamics - Bone: dyn_newhair_03(ComponentPose=@n5, Alpha=@n4, ExternalForce=@n1) {
    angularDampingOverride=1
    angularSpringConstant=50
    bAngularSpring=true
    bGravityOverrideInSimSpace=true
    bLinearSpring=true
    bOverrideAngularDamping=true
    gravityScale=0.35
    linearSpringConstant=50
    simulationSpace=RootRelative
    alphaBoolBlend {blendOption=Linear}
    boundBone {boneName=dyn_newhair_03}
    componentAppliedLinearAccClamp=<1000 400 400>
    componentLinearAccScale=<24 24 50>
    componentLinearVelScale=<24 24 50>
    physicsBodyDefinitions [{
      sphereCollisionRadius=0
      boundBone {boneName=dyn_newhair_03}
      boxExtents=<5 5 5>
      constraintSetup {angularLimitsMax=<20 20 20>, angularLimitsMin=<-20 -20 -20>, angularTarget=<1 0 0>}
      localJointOffset=<2 0 0>
     }]
   }
   n7 = Anim Dynamics - Bone: dyn_newhair_04(ComponentPose=@n6, Alpha=@n4, ExternalForce=@n1) {
    angularDampingOverride=1
    angularSpringConstant=50
    bAngularSpring=true
    bGravityOverrideInSimSpace=true
    bLinearSpring=true
    bOverrideAngularDamping=true
    gravityScale=0.35
    linearSpringConstant=50
    simulationSpace=RootRelative
    alphaBoolBlend {blendOption=Linear}
    boundBone {boneName=dyn_newhair_04}
    componentAppliedLinearAccClamp=<1000 400 400>
    componentLinearAccScale=<24 24 50>
    componentLinearVelScale=<24 24 50>
    physicsBodyDefinitions [{
      sphereCollisionRadius=0
      boundBone {boneName=dyn_newhair_04}
      boxExtents=<5 5 5>
      constraintSetup {angularLimitsMax=<20 20 20>, angularLimitsMin=<-20 -20 -20>, angularTarget=<1 0 0>}
      localJointOffset=<2 0 0>
     }]
   }
   n8 = Anim Dynamics - Bone: dyn_newhair_05(ComponentPose=@n7, Alpha=@n4, ExternalForce=@n1) {
    angularDampingOverride=1
    angularSpringConstant=50
    bAngularSpring=true
    bGravityOverrideInSimSpace=true
    bLinearSpring=true
    bOverrideAngularDamping=true
    gravityScale=0.35
    linearSpringConstant=50
    simulationSpace=RootRelative
    alphaBoolBlend {blendOption=Linear}
    boundBone {boneName=dyn_newhair_05}
    componentAppliedLinearAccClamp=<1000 400 400>
    componentLinearAccScale=<24 24 50>
    componentLinearVelScale=<24 24 50>
    physicsBodyDefinitions [{
      sphereCollisionRadius=0
      boundBone {boneName=dyn_newhair_05}
      boxExtents=<5 5 5>
      constraintSetup {angularLimitsMax=<20 20 20>, angularLimitsMin=<-20 -20 -20>, angularTarget=<1 0 0>}
      localJointOffset=<2 0 0>
     }]
   }
   n9 = Anim Dynamics - Bone: dyn_newhair_06(ComponentPose=@n8, Alpha=@n4, ExternalForce=@n1) {
    angularDampingOverride=1
    angularSpringConstant=50
    bAngularSpring=true
    bGravityOverrideInSimSpace=true
    bLinearSpring=true
    bOverrideAngularDamping=true
    gravityScale=0.35
    linearSpringConstant=50
    simulationSpace=RootRelative
    alphaBoolBlend {blendOption=Linear}
    boundBone {boneName=dyn_newhair_06}
    componentAppliedLinearAccClamp=<1000 400 400>
    componentLinearAccScale=<24 24 50>
    componentLinearVelScale=<24 24 50>
    physicsBodyDefinitions [{
      sphereCollisionRadius=0
      boundBone {boneName=dyn_newhair_06}
      boxExtents=<5 5 5>
      constraintSetup {angularLimitsMax=<20 20 20>, angularLimitsMin=<-20 -20 -20>, angularTarget=<1 0 0>}
      localJointOffset=<2 0 0>
     }]
   }
   n10 = Anim Dynamics - Bone: dyn_newhair_08(ComponentPose=@n9, Alpha=@n4, ExternalForce=@n1) {
    angularDampingOverride=1
    angularSpringConstant=50
    bAngularSpring=true
    bGravityOverrideInSimSpace=true
    bLinearSpring=true
    bOverrideAngularDamping=true
    gravityScale=0.35
    linearSpringConstant=50
    simulationSpace=RootRelative
    alphaBoolBlend {blendOption=Linear}
    boundBone {boneName=dyn_newhair_08}
    componentAppliedLinearAccClamp=<1000 400 400>
    componentLinearAccScale=<24 24 50>
    componentLinearVelScale=<24 24 50>
    physicsBodyDefinitions [{
      sphereCollisionRadius=0
      boundBone {boneName=dyn_newhair_08}
      boxExtents=<5 5 5>
      constraintSetup {angularLimitsMax=<20 20 20>, angularLimitsMin=<-20 -20 -20>, angularTarget=<1 0 0>}
      localJointOffset=<2 0 0>
     }]
   }
   n11 = Anim Dynamics - Bone: dyn_newhair_09(ComponentPose=@n10, Alpha=@n4, ExternalForce=@n1) {
    angularDampingOverride=1
    angularSpringConstant=50
    bAngularSpring=true
    bGravityOverrideInSimSpace=true
    bLinearSpring=true
    bOverrideAngularDamping=true
    gravityScale=0.35
    linearSpringConstant=50
    simulationSpace=RootRelative
    alphaBoolBlend {blendOption=Linear}
    boundBone {boneName=dyn_newhair_09}
    componentAppliedLinearAccClamp=<1000 400 400>
    componentLinearAccScale=<24 24 50>
    componentLinearVelScale=<24 24 50>
    physicsBodyDefinitions [{
      sphereCollisionRadius=0
      boundBone {boneName=dyn_newhair_09}
      boxExtents=<5 5 5>
      constraintSetup {angularLimitsMax=<20 20 20>, angularLimitsMin=<-20 -20 -20>, angularTarget=<1 0 0>}
      localJointOffset=<2 0 0>
     }]
   }
   n12 = Anim Dynamics - Bone: dyn_newhair_010(ComponentPose=@n11, Alpha=@n4, ExternalForce=@n1) {
    angularDampingOverride=1
    angularSpringConstant=50
    bAngularSpring=true
    bGravityOverrideInSimSpace=true
    bLinearSpring=true
    bOverrideAngularDamping=true
    gravityScale=0.35
    linearSpringConstant=50
    simulationSpace=RootRelative
    alphaBoolBlend {blendOption=Linear}
    boundBone {boneName=dyn_newhair_010}
    componentAppliedLinearAccClamp=<1000 400 400>
    componentLinearAccScale=<24 24 24>
    componentLinearVelScale=<24 24 24>
    physicsBodyDefinitions [{
      sphereCollisionRadius=0
      boundBone {boneName=dyn_newhair_010}
      boxExtents=<5 5 5>
      constraintSetup {angularLimitsMax=<5 5 5>, angularLimitsMin=<-5 -5 -5>, angularTarget=<1 0 0>}
      localJointOffset=<2 0 0>
     }]
   }
   n13 = Anim Dynamics - Bone: dyn_newhair_011(ComponentPose=@n12, Alpha=@n4, ExternalForce=@n1) {
    angularDampingOverride=1
    angularSpringConstant=50
    bAngularSpring=true
    bGravityOverrideInSimSpace=true
    bLinearSpring=true
    bOverrideAngularDamping=true
    gravityScale=0.35
    linearSpringConstant=50
    simulationSpace=RootRelative
    alphaBoolBlend {blendOption=Linear}
    boundBone {boneName=dyn_newhair_011}
    componentAppliedLinearAccClamp=<1000 400 400>
    componentLinearAccScale=<24 24 50>
    componentLinearVelScale=<24 24 50>
    physicsBodyDefinitions [{
      sphereCollisionRadius=0
      boundBone {boneName=dyn_newhair_011}
      boxExtents=<5 5 5>
      constraintSetup {angularLimitsMax=<20 20 20>, angularLimitsMin=<-20 -20 -20>, angularTarget=<1 0 0>}
      localJointOffset=<2 0 0>
     }]
   }
   n14 = RigidBody(ComponentPose=@n13, Alpha=Get rigidDanglesAlpha()) {
    bClampLinearTranslationLimitToRefPose=true
    bOverrideWorldGravity=true
    bTransferBoneVelocities=true
    overridePhysicsAsset=&/F_MED_Butterfly_FaceAcc_Physics
    alphaBoolBlend {blendOption=Linear}
    baseBoneRef {boneName=head}
    componentAppliedLinearAccClamp=<1000 1000 1000>
    componentLinearAccScale=<1 1 1>
    componentLinearVelScale=<1 1 1>
    overrideWorldGravity {z=-980}
   }
   Output Pose(Result=Component To Local(ComponentPose=@n14)) {layerGroup=None}
  ]}
 EventGraph {nodes [
   Event BlueprintUpdateAnimation()
   n1 = "To Float (Boolean)"(InBool=Get bIsTargeting())
   n2 = Get bIsJumping()
   > Set rigidDanglesAlpha(rigidDanglesAlpha:double=MapRangeClamped(
     Value="To Float (Boolean)"(InBool=@n2)
     InRangeA=0
     InRangeB=1
     OutRangeA=MapRangeClamped(Value=@n1, InRangeA=0, InRangeB=1, OutRangeA=0.45, OutRangeB=0.2)
     OutRangeB=0.35
    )) -> (Output_Get:double)
   > n3 = Branch(Condition=Get bIsJumping())
   > Delay(Duration=0.25)
   > Set jumpForce(jumpForce:Vector="0,0,2500") -> (Output_Get:Vector)
   > n4 = Delay(Duration=0.375)
   Do Once(execute=@n3.else, Reset=@n4.then)
   > Set jumpForce(jumpForce="0, 0, 0")
  ]}
]
variables [
 jumpForce {type=Vector}
 rigidDanglesAlpha {type=number}
]
